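# Post-copy animation blueprint for the Emperor male medium player body part, adding deformation and a simulation-space bone to the pose copied from its attached parent mesh.
# The copied pose runs through twist and sculpter Control Rigs, then dyn_root_simspace's rotation is replaced in world space.
# Each update, a deprecated sim-space function computes that rotation, feeding its interpolation and space-lock state back through variables.
# A joint debug draw node sits unconnected.
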AnimBlueprint /Game/Characters/Player/Male/Medium/Bodies/M_MED_Emperor/Meshes/M_MED_Emperor_AnimBP
target_skeleton=&/M_MED_Emperor_Skeleton
graphs [
 AnimGraph {nodes [
   n1 = Control Rig(Source=Copy Pose From Mesh() {bUseAttachedParent=true}) {
    alphaBoolBlend {blendOption=Linear}
    controlRigAssetReference {blueprintRigClass=&/M_MED_Emperor_Deform_CtrlRig_twists.M_MED_Emperor_Deform_CtrlRig_twists_C}
   }
   n2 = Control Rig!(
    Source=@n1
    Alpha
    KneeRot_fr_l_114deg:Vector
    KneeTrans_fr_l_114deg:Vector="0.258,-3.591,0.009"
    KneeScale_fr_l_114deg:Vector="1,1,1"
    KneeScale_fr_l_130deg:Vector="1,1,1"
    KneeRot_fr_l_130deg:Vector="0,0,-51.2"
    KneeTrans_fr_l_130deg:Vector="0.258,-3.591,0.009"
   ) -> (Pose) {
    alphaBoolBlend {blendOption=Linear}
    controlRigAssetReference {blueprintRigClass=&/M_MED_Emperor_Deform_CtrlRig_sculpters.M_MED_Emperor_Deform_CtrlRig_sculpters_C}
   }
   n3 = "Transform (Modify) Bone - Bone: dyn_root_simspace"(
    ComponentPose=Local To Component(LocalPose=@n2)
    Alpha=1
    Translation="0,0,0"
    Rotation=Get sim_space_rotation() -> (sim_space_rotation:Rotator)
    Scale="1,1,1"
   ) {
    rotationMode=BMM_Replace
    rotationSpace=BCS_WorldSpace
    translationSpace=BCS_BoneSpace
    alphaBoolBlend {blendOption=Linear}
    alphaScaleBiasClamp {bInterpResult=true, interpSpeedDecreasing=2, interpSpeedIncreasing=2}
    boneToModify {boneName=dyn_root_simspace}
   }
   Output Pose(Result=Component To Local(ComponentPose=@n3)) {layerGroup=None}
  ]}
 EventGraph {nodes [
   n1 = Event BlueprintUpdateAnimation()
   > n2 = ComplexSimSpaceFunction DEPRECATED!(
    execute
    CharacterPartAnimInstance:Custom Character Part Anim Instance Object Reference=Self-Reference()
    InterpSpeedMax:double=9
    PawnSpeedMax:double=20
    RInterpErrorTolerance:double=1
    DeltaTime:double=@n1.DeltaTimeX
    DebugDraw:bool=false
    sim_space_rotation_in:Rotator=Get sim_space_rotation()
    inComponentSpaceIn:bool=Get InComponentSpace() -> (InComponentSpace:bool)
    inComponentSpaceLastFrameIn:bool=Get InComponentSpaceLastFrame() -> (InComponentSpaceLastFrame:bool)
    lockedOnSimSpaceIn:bool=Get LockedOnSimSpace() -> (LockedOnSimSpace:bool)
    InterpSpeedIn:double=Get InterpSpeed() -> (InterpSpeed:double)
   ) -> (
    then
    sim_space_rotation:Rotator
    inComponentSpaceOut:bool
    inComponentSpaceLastFrameOut:bool
    lockedOnSimSpaceOut:bool
    InterpSpeedOut:double
   )
   > Set sim_space_rotation(sim_space_rotation:Rotator=@n2) -> (Output_Get:Rotator)
   > Set InComponentSpace(InComponentSpace:bool=@n2.inComponentSpaceOut) -> (Output_Get:bool)
   > Set InComponentSpaceLastFrame(InComponentSpaceLastFrame:bool=@n2.inComponentSpaceLastFrameOut) -> (Output_Get:bool)
   > Set LockedOnSimSpace(LockedOnSimSpace:bool=@n2.lockedOnSimSpaceOut) -> (Output_Get:bool)
   > Set InterpSpeed(InterpSpeed:double=@n2.InterpSpeedOut) -> (Output_Get:double)
   DrawJointDebugFunction!(
    execute
    SkelMeshComponent:Scene Component Object Reference
    JointOffset:Vector="0, 0,10"
    JointName:Array of Names=Make Array("[0]":Name=dyn_root_simspace) -> (Array:Array of Names)
    Scale:double=10
    Duration:double=0
    Thickness:double=0.7
    DebugDraw:bool=true
    Foreground:EDrawDebugSceneDepthPriorityGroup Enum=Foreground
   ) -> (then)
  ]}
]
variables [
 InComponentSpace {type=boolean}
 InComponentSpaceLastFrame {type=boolean}
 InterpSpeed {type=number}
 LockedOnSimSpace {type=boolean}
 sim_space_rotation {type=Rotator}
]
